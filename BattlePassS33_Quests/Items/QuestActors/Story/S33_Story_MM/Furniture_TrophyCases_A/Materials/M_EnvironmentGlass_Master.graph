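# Master translucent environment glass material, stored with quest trophy case props.
# Base color layers static or time-of-day tint, cubemap or skylight reflections, an optional mask, grime and label texture.
# Fresnel drives opacity; optional blur fakes refraction by offsetting scene color with a tiled normal and forcing opacity to one, falling back to glass color on mobile.
# Optional normal map and emissive join these in material attributes, passed through a destructible damage function.
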Material /BattlePassS33_Quests/Items/QuestActors/Story/S33_Story_MM/Furniture_TrophyCases_A/Materials/M_EnvironmentGlass_Master
graph {
 properties {
  bAutomaticallySetUsageInEditor=false
  bScreenSpaceReflections=true
  bUseMaterialAttributes=true
  bUsedWithInstancedStaticMeshes=true
  bUsedWithNanite=true
  blendMode=BLEND_Translucent
  pixelDepthOffsetMode=PDOM_Legacy
  refractionMethod=RM_PixelNormalOffset
  translucencyLightingMode=TLM_VolumetricDirectional
  usedShadingModels=Default Lit
 }
 nodes [
  n1 = NamedRerouteUsage()
  n2 = StaticSwitchParameter(True=Multiply(A=NamedRerouteUsage(), B=@n1), False=@n1) {
   desc="A fake refraction that blurs the scene behind the glass.  Doesn't work on mobile so will fallback to glass color."
   group=Blur
   parameterName=UseGlassBlur
  }
  n3 = LinearInterpolate(
   A=@n2
   B=@n1
   Alpha=LinearInterpolate(
    A=ScalarParameter() {defaultValue=0.15, group=Opacity, parameterName=Opacity}
    B=ScalarParameter() {defaultValue=0.5, group=Opacity, parameterName=Opacity_Fresnel}
    Alpha=NamedRerouteUsage()
   )
  )
  n4 = ShadingPathSwitch(Default=MaterialProxyReplace(Realtime=FeatureLevelSwitch(Default=@n3, ES3_1=@n1), MaterialProxy=NamedRerouteUsage()))
  n5 = TextureSampleParameter2D() {group=DiffuseMask, parameterName=DiffuseMaskedTexture, texture=/Game/Global/Textures/Default/Blanks/T_Blank_LegalWhite_sRGBColor}
  n6 = StaticSwitchParameter(True=LinearInterpolate(A=@n4, B=@n5.RGB, Alpha=@n5.RGB), False=@n4) {
   desc="An optional Diffuse Texture with Alpha being a mask between normal glass and diffuse texture.  Useful for labels on bottles for example"
   group=DiffuseMask
   parameterName=UseDiffuseMask
  }
  n7 = LinearInterpolate(
   A=ScalarParameter() {defaultValue=0.15, group=Opacity, parameterName=Opacity}
   B=ScalarParameter() {defaultValue=0.5, group=Opacity, parameterName=Opacity_Fresnel}
   Alpha=NamedRerouteUsage()
  )
  n8 = ScalarParameter() {defaultValue=1, parameterName=MaskOpacity}
  n9 = VectorParameter() {group=Mask, parameterName=MaskColor, channelNames {a=BlendMask, b="", g="", r=""}, defaultValue=<rgba 1 1 1 1>}
  n10 = Multiply(A=@n9.BlendMask, B=NamedRerouteUsage())
  n11 = StaticSwitchParameter(True=Saturate(None=LinearInterpolate(A=@n7, B=@n8, Alpha=@n10)), False=@n7) {
   desc="The Mask is a general use separate color blended with a texture in the Mask Texture's Blue channel.  Example is the lines running through break resistent glass."
   group=Mask
   parameterName=UseMask
  }
  n12 = StaticSwitchParameter(
   True=LinearInterpolate(A=@n11, B=ScalarParameter() {defaultValue=0.8, group=Opacity, parameterName=Grime_Opacity}, Alpha=NamedRerouteUsage())
   False=@n11
  ) {desc="Grime is typically for dirt.  Grim has separate color and roughness", group=Grime, parameterName=UseGrimeTexture}
  n13 = StaticSwitchParameter(True=Constant() {desc=We make opaque and use the scene texture to blue behind translucency in the diffuse, r=1}, False=@n12) {
   desc="A fake refraction that blurs the scene behind the glass.  Doesn't work on mobile so will fallback to glass color."
   group=Blur
   parameterName=UseGlassBlur
  }
  n14 = ShadingPathSwitch(Default=@n13, Mobile=@n12)
  n15 = ScalarParameter() {defaultValue=1, desc=Opacity of the White area of the DiffuseMaskedTexture alpha, group=DiffuseMask, parameterName=DiffuseMaskedOpacity}
  n16 = StaticSwitchParameter(True=LinearInterpolate(A=@n14, B=@n15, Alpha=@n5.A), False=@n14) {
   desc="An optional Diffuse Texture with Alpha being a mask between normal glass and diffuse texture.  Useful for labels on bottles for example"
   group=DiffuseMask
   parameterName=UseDiffuseMask
  }
  n17 = ScalarParameter() {defaultValue=1, group=NormalMap, parameterName=NormalMapStrength}
  n18 = MaterialFunctionCall(R=@n17, G=@n17, B=Constant() {r=1}) {materialFunction=&AppendMany}
  n19 = AppendVector(
   A=ScalarParameter() {defaultValue=1, group=NormalMap, parameterName=NormalMapU_Scale}
   B=ScalarParameter() {defaultValue=1, group=NormalMap, parameterName=NormalMapV_Scale}
  )
  n20 = TextureCoordinate()
  n21 = TextureSampleParameter2D(UVs=StaticSwitchParameter(True=Multiply(A=@n19, B=@n20), False=@n20) {group=NormalMap, parameterName=AdjustNormalMapUVs}) {
   desc=Just a good ole normal map for mapping normals
   group=NormalMap
   parameterName=NormalTexture
   samplerType=SAMPLERTYPE_Normal
   texture=&FlatNormalBlue
  }
  n22 = StaticSwitchParameter(True=Multiply(A=Normalize(VectorInput=@n18.RGB), B=@n21.RGB), False=@n21.RGB) {group=NormalMap, parameterName=AdjustNormalMapStrength}
  n23 = Constant3Vector() {constant=<rgba 0 0 1 0>}
  n24 = NamedRerouteUsage()
  n25 = ScalarParameter() {parameterName=EmissiveRoughnessMaskIntensity}
  n26 = Desaturation(None=Multiply(A=@n6, B=LinearInterpolate(B=@n24, Alpha=@n25) {constA=1}), Fraction=ScalarParameter() {parameterName=EmissiveDesaturation})
  n27 = SetMaterialAttributes!(
   MaterialAttributes
   Base Color=@n6
   Opacity=@n16
   Normal=ShadingPathSwitch(
    Default=StaticSwitchParameter(True=@n22, False=@n23) {desc=Use a normal map for lighting, group=NormalMap, parameterName=UseNormalMap}
    Mobile=@n23
   )
   Emissive Color=StaticSwitchParameter(True=EyeAdaptationInverse(LightValueInput=@n26).EyeAdaptationInverse, False=Constant()) {group=DiffuseMask, parameterName=AddDiffuseToEmissive}
   Refraction=ScalarParameter() {defaultValue=1, parameterName=Refraction}
  ) {attributeSetTypes [
    #1
    #2
    #3
    #4
    #5
   ]}
  MaterialFunctionCall(Material_In=@n27) {materialFunction=&MF_FN_DestructibleDamage}
  n28 = TextureSampleParameter2D(UVs=Multiply(A=TextureCoordinate(), B=ScalarParameter() {defaultValue=12, group=Blur, parameterName=BlurNormalTile})) {
   group=Blur
   parameterName=BlurNormalMap
   samplerType=SAMPLERTYPE_Normal
   texture=/Game/Global/Textures/Default/Tileable_Patterns/T_SkinBump_N
  }
  n29 = ComponentMask(None=@n28.RGB) {g=true, r=true}
  n30 = ScalarParameter() {defaultValue=0.1, group=Blur, parameterName=BlurAmount, sortPriority=19}
  NamedRerouteDeclaration(None=SceneColor!(OffsetFraction=Multiply(A=@n29, B=@n30)) {inputMode=OffsetFraction}.RGB) {name=DistortedSceneColor, nodeColor=<rgba 0.4062 1 0 1>}
  n31 = ScalarParameter() {desc=Adjust the value for the black part of grime mask, group=Grime, parameterName=Grime_AdjustMin}
  n32 = ScalarParameter() {defaultValue=0.8, desc=Adjust the value for the white part of grime mask, group=Grime, parameterName=Grime_AdjustMax}
  n33 = ScalarParameter() {defaultValue=1, group=Grime, parameterName=Grime_RepeatU}
  n34 = ScalarParameter() {defaultValue=1, group=Grime, parameterName=Grime_RepeatV}
  n35 = TextureSampleParameter2D(UVs=Multiply(A=TextureCoordinate(), B=AppendVector(A=@n33, B=@n34))) {
   group=Grime
   parameterName=TextureMask
   samplerType=SAMPLERTYPE_Masks
   texture=/Game/Global/Materials/EnvironmentGlass/T_GlassTestMask_Indestruct
   channelNames {a="", b=Mask, g=Roughness, r=Grime}
  }
  n36 = StaticSwitchParameter(True=OneMinus(None=@n35.Grime), False=@n35.Grime) {desc="Inverts the grime mask in Red channel of TextureMask sample.  ", group=Grime, parameterName=InvertGrime}
  NamedRerouteDeclaration(None=Saturate(None=LinearInterpolate(A=@n31, B=@n32, Alpha=@n36) {constA=1})) {name=GrimeMask, nodeColor=<rgba 0.4583 0.2142 0.0789 1>}
  n37 = ScalarParameter() {defaultValue=1, group=Mask, parameterName=Mask_RepeatU}
  n38 = ScalarParameter() {defaultValue=1, group=Mask, parameterName=Mask_RepeatV}
  n39 = TextureSampleParameter2D(UVs=Multiply(A=TextureCoordinate(), B=AppendVector(A=@n37, B=@n38))) {
   group=Grime
   parameterName=TextureMask
   samplerType=SAMPLERTYPE_Masks
   texture=&T_GlassTestMask_Indestruct
   channelNames {a="", b=Mask, g=Roughness, r=Grime}
  }
  NamedRerouteDeclaration(None=@n39.Mask) {name=MaskMask, nodeColor=<rgba 0.7135 0.3999 0.6805 1>}
  n40 = ReflectionVectorWS()
  n41 = ComponentMask(None=@n40) {g=true, r=true}
  n42 = TextureSampleParameterCube!(UVs=@n40, Apply View MipBias) {
   group=CubeMap
   parameterName=CubemapTexture
   samplerType=SAMPLERTYPE_LinearColor
   texture=/Game/Creative/Environments/Meshes/Glass/Textures/T_Glass_Cubemap_01
  }
  n43 = SkyLightEnvMapSample(Direction=@n40, Roughness=ScalarParameter() {defaultValue=0.1, group=CubeMap, parameterName=SkylightEnvMapRoughness})
  n44 = ScalarParameter() {
   defaultValue=0.5
   desc="A multiplier on the cubemap to blend it in and out.  Higher values show it more."
   group=CubeMap
   parameterName=CubeMap_Mult
  }
  n45 = ScalarParameter() {
   defaultValue=0.8
   desc=When the angle of the surface faces more away from the camera this will be the value used for Cubemap mult.
   group=CubeMap
   parameterName=CubeMap_Fresnel
  }
  n46 = NamedRerouteUsage()
  n47 = ScalarParameter() {defaultValue=0.5, group=Roughness, parameterName=RoughMask_Value_MobileForward}
  n48 = Multiply(
   A=NamedRerouteUsage()
   B=ShadingPathSwitch(Default=ScalarParameter() {defaultValue=0.5, group=Roughness, parameterName=RoughMask_Value}, Forward=@n47, Mobile=@n47)
  )
  n49 = Multiply(
   A=LinearInterpolate(A=LinearInterpolate(A=@n44, B=@n45, Alpha=@n46), Alpha=Saturate(None=@n48)) {constB=0}
   B=StaticSwitchParameter(
    True=TextureSampleParameter2D(UVs=@n41) {group=CubeMap, parameterName=CubemapTexture2d, texture=/Game/Packages/Fortress_SharedMaterials/Textures/T_Cubemap_01_tex}.RGB
    False=StaticSwitchParameter(True=@n42.RGB, False=@n43) {group=CubeMap, parameterName=UseTextureCubeMap}
   ) {group=CubeMap, parameterName=UseTextureCubeMap2d}
  )
  n50 = VectorParameter() {group=CubeMap, parameterName=CubemapColorScale, defaultValue=<rgba 1 1 1 0>}
  NamedRerouteDeclaration(None=StaticSwitchParameter(True=Multiply(A=@n49, B=@n50.RGB), False=Constant()) {defaultValue=true, group=CubeMap, parameterName=UseCubeMap}) {name=CubeMapColor, nodeColor=<rgba 0.2422 0 1 1>}
  n51 = StaticSwitchParameter(
   True=MaterialFunctionCall(
    Color_Morning=VectorParameter() {group=Glass Color, parameterName=GlassColorMorning, defaultValue=<rgba 0.3712 0.5035 0.6198 1>}.RGB
    Color_Day=VectorParameter() {group=Glass Color, parameterName=GlassColorDay, defaultValue=<rgba 0.3712 0.5035 0.6198 1>}.RGB
    Color_Evening=VectorParameter() {group=Glass Color, parameterName=GlassColorEvening, defaultValue=<rgba 0.3712 0.5035 0.6198 1>}.RGB
    Color_Night=VectorParameter() {group=Glass Color, parameterName=GlassColorNight, defaultValue=<rgba 0.3712 0.5035 0.6198 1>}.RGB
   ) {materialFunction=&Color_Time_of_Day}.Result
   False=VectorParameter() {group=Glass Color, parameterName=GlassColor, defaultValue=<rgba 0.3712 0.5035 0.6198 1>}.RGB
  ) {group=Glass Color, parameterName=UseTODGlassColor}
  n52 = Multiply(A=@n51, B=ScalarParameter() {defaultValue=1, group=Glass Color, parameterName=MobileLowGlassColorScale})
  n53 = ShadingPathSwitch(Default=@n51, Forward=@n52, Mobile=@n52)
  n54 = Add(
   A=NamedRerouteDeclaration(None=QualitySwitch(Default=@n53, Low=@n52)) {name=GlassColor, nodeColor=<rgba 0 1 0.8359 1>}
   B=NamedRerouteUsage()
  )
  n55 = VectorParameter() {group=Mask, parameterName=MaskColor, channelNames {a=BlendMask, b="", g="", r=""}, defaultValue=<rgba 1 1 1 1>}
  n56 = Multiply(A=@n55.BlendMask, B=NamedRerouteUsage())
  n57 = StaticSwitchParameter(True=LinearInterpolate(A=@n54, B=@n55, Alpha=Saturate(None=@n56)), False=@n54) {
   desc="The Mask is a general use separate color blended with a texture in the Mask Texture's Blue channel.  Example is the lines running through break resistent glass."
   group=Mask
   parameterName=UseMask
  }
  NamedRerouteDeclaration(None=@n57) {name=GlassColorWithMask, nodeColor=<rgba 1 0.0938 0 1>}
  n58 = ScalarParameter() {defaultValue=1, group=Roughness, parameterName=Rough_RepeatU}
  n59 = ScalarParameter() {defaultValue=1, group=Roughness, parameterName=Rough_RepeatV}
  n60 = TextureSampleParameter2D(UVs=Multiply(A=TextureCoordinate(), B=AppendVector(A=@n58, B=@n59))) {
   group=Grime
   parameterName=TextureMask
   samplerType=SAMPLERTYPE_Masks
   texture=&T_GlassTestMask_Indestruct
   channelNames {a="", b=Mask, g=Roughness, r=Grime}
  }
  NamedRerouteDeclaration(None=@n60.Roughness) {name=RoughMask, nodeColor=<rgba 0.602 0.7917 0.7205 1>}
  n61 = ScalarParameter() {
   defaultValue=1
   desc="Tighten/loosen the fresnel  blend.  Fresnel allows for different values of opacity and roughness based on the angle of the surface facing the camera. (Glass is shinier when looking at it from the edge)"
   parameterName=Fresnel Exponent
  }
  NamedRerouteDeclaration(None=Fresnel(ExponentIn=@n61) {baseReflectFraction=0}) {name=FresnelMask, nodeColor=<rgba 1 0.8814 0 1>}
  n62 = NamedRerouteUsage()
  n63 = VectorParameter() {group=Grime, parameterName=GrimeColor, defaultValue=<rgba 0.3712 0.5035 0.6198 1>}
  n64 = NamedRerouteUsage()
  NamedRerouteDeclaration(None=StaticSwitchParameter(True=LinearInterpolate(A=@n62, B=@n63.RGB, Alpha=@n64), False=NamedRerouteUsage()) {desc="Grime is typically for dirt.  Grim has separate color and roughness", group=Grime, parameterName=UseGrimeTexture}) {name=GlassColorFinal, nodeColor=<rgba 1 0 0.9375 1>}
 ]
}
